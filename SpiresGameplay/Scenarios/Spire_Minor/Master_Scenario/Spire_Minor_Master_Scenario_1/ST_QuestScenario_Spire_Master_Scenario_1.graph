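# StateTree for a minor Spire quest scenario: fight Spire-faction AI, finish a sub-scenario, then destroy the spire for gold.
# It spawns combatants and a sub-scenario actor, then on player entry makes the spire invulnerable and combatants hostile until the sub-scenario completes.
# The vulnerable spire then faces looping, delayed reinforcement waves, re-entered through a tiny "HACK Delay" state.
# A parent-state destruction listener breaks out to reward and completion.
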StateTree /SpiresGameplay/Scenarios/Spire_Minor/Master_Scenario/Spire_Minor_Master_Scenario_1/ST_QuestScenario_Spire_Master_Scenario_1
global_tasks [
 {
  id=#1
  description=Get Scenario Actor - MinorSpire
  properties {
   iD=#2
   instance {
    _structType=&GetEncounterActorStateTreeTaskInstanceData
    actorQuery {
     autoDescription=" ANY( QuestEncounter.Spire_Minor_Random.Encounter.Target.Spire )"
     tagDictionary [{tagName=QuestEncounter.Spire_Minor_Random.Encounter.Target.Spire}]
    }
   }
   node {name=Get Scenario Actor - MinorSpire, _structType=&GetEncounterActorStateTreeTask}
  }
 }
 {
  id=#3
  description="Get Variable [Actor Class]"
  properties {
   iD=#4
   instance {
    _structType=/Script/EncountersRuntime.EncounterGetVariableActorClassStateTreeTaskInstanceData
    userActor=null
    value=null
    variableIdentifier {tagName=QuestEncounter.Spire_Minor_Combat_1.Encounter.Variable.SubScenario}
   }
   node {
    name=None
    _structType=/Script/EncountersRuntime.EncounterGetVariableActorClassStateTreeTask
    bCanEditConsideredForCompletion=true
    bConsideredForCompletion=true
    bHasTaskCompletionDelegateDispatcher=false
    bTaskEnabled=true
    transitionHandlingPriority=Normal
    bindingsBatch {value=65535}
    executionRuntimeTemplateIndex {value=65535}
    instanceDataHandle {index=65535, source=None, stateHandle {index=65535}}
    instanceTemplateIndex {value=65535}
    outputBindingsBatch {value=65535}
   }
  }
 }
]
states [
 Root {
  id=#5
  type=STATE
  children [#6, #7, #8, #9]
  tasks []
  transitions []
 }
 "Init: Spawn initial combatants, Spawn random SubScenario" {
  id=#6
  type=STATE
  tasks [
   {
    id=#10
    description="LWM Encounter [Dynamic Start]"
    properties {
     iD=#11
     instance {
      _structType=/Script/EncountersRuntime.LWMEncounterDynamicStartStateTreeTaskInstanceData
      bAttachVolumeToSpawn=false
      bOverrideAnchorPosition=false
      bOverrideDelayBetweenEvents=false
      bOverrideEnvQuery=false
      bOverrideStageActorMaxCount=false
      bOverrideStageActorMaxSpawnedCount=false
      bPauseOnExit=false
      bSpecifyFaction=true
      bUseVolumeAsAnchor=true
      bWaitForEncounterDoneSpawning=true
      delayBetweenEventsMax=0
      delayBetweenEventsMin=0
      dynamicEmptyEncounter=/SpireSharedAssets/Lager/Lager_Encounter_Dynamic_Empty
      faction=/FortQuestEncounters/Factions/Spires/Faction_Spires
      instigator=null
      owningActor=null
      spawnAroundEnvironmentQuery=null
      spawnAroundEnvironmentQueryRunMode=AllMatching
      spawnAroundNavigationSourceOverride=""
      stageActorMaxCount=1
      stageActorMaxSpawnedCount=-1
      userActor=null
      actorTags {gameplayTags []}
      anchorPointTagQuery {autoDescription="", tokenStreamVersion=0, userDescription="", queryTokenStream [], tagDictionary []}
      dynamicEncounterEntries [{
        actorCountOverrideMax=1
        actorCountOverrideMin=1
        bCreateRandomSingleActorEvents=false
        bOverrideActorCount=true
        bUseEncounterTagQuery=true
        bUseEventTemplate=true
        debugName=Test SubScenario Combatant
        maxConcurrentSpawns=1
        noActorLimitEventCount=1
        spawnCount=1
        actorDescriptions []
        encounterTagQuery {autoDescription="", tokenStreamVersion=0, userDescription="", queryTokenStream [], tagDictionary []}
        eventTemplate {
         activationGamePhaseStep=None
         activationMaxDelay=0
         activationMinDelay=0
         activationSafeZoneIndex=0
         bCanSpawnWithoutMatchingPlayerDistance=false
         bCheckForActorDensity=true
         bOnlyInNextSafeZone=false
         bOnlyInNextSafeZoneIfStormShrinking=false
         bOnlyInSafeZone=false
         bShouldProcessEventRequestImmediately=false
         deactivationMaxDelay=-1
         deactivationMinDelay=-1
         maxActorDensityOverride=-1
         maxDistanceToPlayer=-1
         minDistanceToPlayer=-1
         randomActivationChance=100
         weight=1
         calendarEventConditions {isActiveWithoutSeasonalManager=false, conditions []}
         canSpawnCondition {description=Empty, schemaClass=/Script/LagerRuntime.FortAthenaLivingWorldConditionSchema, editableConditions [], sharedDefinition {}}
         despawnCondition {
          bCanBeDespawned=false
          bCanDespawnInTheStorm=false
          bCanDespawnWhenInCombatWithPlayer=false
          canStartDespawningPhaseStep=None
          canStartDespawningSafeZonePhase=-1
          maxStimulusAgeToBeConsideredInCombat=30
          minDistanceToPlayer=-1
          timeAsCandidateBeforeDespawn=10
         }
         despawnConditions []
         eventDatas []
         eventInstanceMaxCount [-1]
         eventInstanceMaxSpawnedCount [-1]
         instantDeathMaxRespawnTime []
         instantDeathMinRespawnTime []
         isEnabled {registryType=None, value=1, curve {curveTable=null, rowName=None}}
         maxCount [3]
         maxRespawnTime []
         maxSpawnedCount [3]
         minRespawnTime []
         playlistTagQuery {autoDescription="", tokenStreamVersion=0, userDescription="", queryTokenStream [], tagDictionary []}
         tagDensityConditions []
        }
        pointProviderTagQueryOverride {autoDescription="", tokenStreamVersion=0, userDescription="", queryTokenStream [], tagDictionary []}
       }]
      encounterIdentifierTags {gameplayTags [{tagName=QuestEncounter.Spire_Minor_Combat_1.AI.Goal.Assault.Wave.1}]}
      leashRadiusInner {registryType=None, value=1000, curve {curveTable=null, rowName=None}}
      leashRadiusOuter {registryType=None, value=3000, curve {curveTable=null, rowName=None}}
      pointProviderTagQuery {
       autoDescription=" ANY( QuestEncounter.Spire_Minor_Combat_1.Location.SpawnPoint.Wave )"
       tokenStreamVersion=0
       userDescription=Spawning will select a point for each combatant that matches this tag
       queryTokenStream [0,1,1,1,0]
       tagDictionary [{tagName=QuestEncounter.Spire_Minor_Combat_1.Location.SpawnPoint.Wave}]
      }
      spawnActions [{spawnActionClass=/SpiresGameplay/Scenarios/SpawnActions/BP_PlayVFX_SpawnFX.BP_PlayVFX_SpawnFX_C, spawnActionTag {tagName=None}}]
      spawnTimeout {registryType=None, value=-1, curve {curveTable=null, rowName=None}}
     }
     node {
      name=None
      _structType=/Script/EncountersRuntime.LWMEncounterDynamicStartStateTreeTask
      bCanEditConsideredForCompletion=true
      bConsideredForCompletion=true
      bHasTaskCompletionDelegateDispatcher=false
      bTaskEnabled=true
      transitionHandlingPriority=Normal
      bindingsBatch {value=65535}
      executionRuntimeTemplateIndex {value=65535}
      instanceDataHandle {index=65535, source=None, stateHandle {index=65535}}
      instanceTemplateIndex {value=65535}
      outputBindingsBatch {value=65535}
     }
    }
   }
   {
    id=#12
    description=Spawn Actor - Spawn SubScenario as actor
    properties {
     iD=#13
     instance {
      _structType=/Script/EncountersRuntime.SpawnActorStateTreeTaskInstanceData
      actorClass=null
      bAttachEncounterVolume=false
      bEnablePhysics=false
      bUseActorLocation=false
      bWaitForSpawnedActor=false
      owningActor=null
      persistenceBehavior=Destroy
      spawnedActor=null
      targetActor=null
      userActor=null
      actorIdentifierTags {gameplayTags []}
      actorTags {gameplayTags []}
      anchorOffset=<0 0 0>
      anchorQuery {autoDescription="", tokenStreamVersion=0, userDescription="", queryTokenStream [], tagDictionary []}
      existingActorQuery {autoDescription="", tokenStreamVersion=0, userDescription="", queryTokenStream [], tagDictionary []}
     }
     node {
      name=Spawn Actor - Spawn SubScenario as actor
      _structType=/Script/EncountersRuntime.SpawnActorStateTreeTask
      bCanEditConsideredForCompletion=true
      bConsideredForCompletion=true
      bHasTaskCompletionDelegateDispatcher=false
      bTaskEnabled=true
      transitionHandlingPriority=Normal
      bindingsBatch {value=65535}
      executionRuntimeTemplateIndex {value=65535}
      instanceDataHandle {index=65535, source=None, stateHandle {index=65535}}
      instanceTemplateIndex {value=65535}
      outputBindingsBatch {value=65535}
     }
    }
   }
  ]
  transitions [{id=#14, trigger=ON_STATE_COMPLETED, state {id=#15, name=Root, linkType=NEXT_STATE}}]
 }
 Wait for Spire Destruction {
  id=#7
  type=STATE
  children [
   #16
   #17
   #18
   #19
   #20
   #21
   #22
   #23
  ]
  tasks [{
    id=#24
    description=Scenario Verb Transition
    properties {
     iD=#25
     instance {
      _structType=/Script/EncountersRuntime.EncounterVerbTransitionStateTreeTaskInstanceData
      bIsTriggered=false
      bUseAdvancedObjectives=false
      objectiveValueThreshold=1
      userActor=null
      objectiveVerbs [{
        _structType=/Script/FortniteGame.FortObjectiveFilter_BuildingWasDestroyed
        author {
         bCanGainProgressForThisObjective=true
         tagQuery {
          autoDescription=" ANY( QuestEncounter.Spire_Minor_Random.Encounter.Target.Spire )"
          tokenStreamVersion=0
          userDescription=""
          queryTokenStream [0,1,1,1,0]
          tagDictionary [{tagName=QuestEncounter.Spire_Minor_Random.Encounter.Target.Spire}]
         }
        }
        context {tagQuery {autoDescription="", tokenStreamVersion=0, userDescription="", queryTokenStream [], tagDictionary []}}
        damageSource {
         bCanGainProgressForThisObjective=true
         tagQuery {autoDescription="", tokenStreamVersion=0, userDescription="", queryTokenStream [], tagDictionary []}
        }
       }]
      objectives []
      onTriggered {iD=#26}
     }
     node {
      name=None
      _structType=/Script/EncountersRuntime.EncounterVerbTransitionStateTreeTask
      bCanEditConsideredForCompletion=true
      bConsideredForCompletion=true
      bHasTaskCompletionDelegateDispatcher=false
      bTaskEnabled=true
      transitionHandlingPriority=Normal
      bindingsBatch {value=65535}
      executionRuntimeTemplateIndex {value=65535}
      instanceDataHandle {index=65535, source=None, stateHandle {index=65535}}
      instanceTemplateIndex {value=65535}
      outputBindingsBatch {value=65535}
      transitionTo {name=Stage_Reward, fallback=None, iD=#27, linkType=GotoState, stateHandle {index=65535}}
     }
    }
   }]
  transitions []
 }
 Stage_Reward {
  id=#8
  type=STATE
  tasks [
   {
    id=#28
    description=Get Anchor - Reward
    properties {
     iD=#29
     instance {
      _structType=/Script/EncountersRuntime.GetAnchorStateTreeTaskInstanceData
      anchorActor=null
      userActor=null
      anchorQuery {
       autoDescription=" ANY( QuestEncounter.Spire_Minor_Combat_1.Location.Reward.1 )"
       tokenStreamVersion=0
       userDescription=""
       queryTokenStream [0,1,1,1,0]
       tagDictionary [{tagName=QuestEncounter.Spire_Minor_Combat_1.Location.Reward.1}]
      }
     }
     node {
      name=Get Anchor - Reward
      _structType=/Script/EncountersRuntime.GetAnchorStateTreeTask
      bCanEditConsideredForCompletion=true
      bConsideredForCompletion=true
      bHasTaskCompletionDelegateDispatcher=false
      bTaskEnabled=true
      transitionHandlingPriority=Normal
      bindingsBatch {value=65535}
      executionRuntimeTemplateIndex {value=65535}
      instanceDataHandle {index=65535, source=None, stateHandle {index=65535}}
      instanceTemplateIndex {value=65535}
      outputBindingsBatch {value=65535}
     }
    }
   }
   {
    id=#30
    description=Spawn Pickup
    properties {
     iD=#31
     instance {
      _structType=/Script/FortniteGame.SpawnPickupStateTreeTaskInstanceData
      instigatorActor=null
      itemDefinition=null
      lootTierName=Loot_GoldBars_XL
      spawnedPickup=null
      targetActor=null
      userActor=null
      flingConeAngle {registryType=None, value=0, curve {curveTable=null, rowName=None}}
      flingDirection=<0 0 1>
      flingMagnitude {registryType=None, value=0, curve {curveTable=null, rowName=None}}
      spawnOffset=<0 0 0>
      spawnedPickupActors []
     }
     node {
      name=None
      _structType=/Script/FortniteGame.SpawnPickupStateTreeTask
      bCanEditConsideredForCompletion=true
      bConsideredForCompletion=true
      bHasTaskCompletionDelegateDispatcher=false
      bTaskEnabled=true
      transitionHandlingPriority=Normal
      bindingsBatch {value=65535}
      executionRuntimeTemplateIndex {value=65535}
      instanceDataHandle {index=65535, source=None, stateHandle {index=65535}}
      instanceTemplateIndex {value=65535}
      outputBindingsBatch {value=65535}
     }
    }
   }
   {
    id=#32
    description=Delay 1.0
    properties {
     iD=#33
     instance {_structType=&StateTreeDelayTaskInstanceData, duration=1}
     node {name=None, _structType=&StateTreeDelayTask}
    }
   }
  ]
  transitions [{id=#34, trigger=ON_STATE_COMPLETED, state {id=#9, name=Complete}}]
 }
 Complete {
  id=#9
  type=STATE
  tasks [
   {
    id=#35
    description=Send Custom Verb
    properties {
     iD=#36
     instance {
      _structType=/Script/EncountersRuntime.EncounterSendCustomVerbStateTreeTaskInstanceData
      bSendVerbForEachPlayerAuthor=false
      userActor=null
      customVerb {amount=1, customVerbTag {tagName=CustomVerb.Encounter.Completed}, subjects [], values []}
     }
     node {
      name=None
      _structType=/Script/EncountersRuntime.EncounterSendCustomVerbStateTreeTask
      bCanEditConsideredForCompletion=true
      bConsideredForCompletion=true
      bHasTaskCompletionDelegateDispatcher=false
      bTaskEnabled=true
      transitionHandlingPriority=Normal
      bindingsBatch {value=65535}
      executionRuntimeTemplateIndex {value=65535}
      instanceDataHandle {index=65535, source=None, stateHandle {index=65535}}
      instanceTemplateIndex {value=65535}
      outputBindingsBatch {value=65535}
     }
    }
   }
   {
    id=#37
    description=Grant Reactions
    properties {
     iD=#38
     instance {
      _structType=/Script/EncountersRuntime.GrantReactionStateTreeTaskInstanceData
      grantType=GrantToParticipants
      targetActor=null
      userActor=null
      reactionEntries []
      reactions [
       {
        _structType=/Script/FortniteGame.FortEventReaction_PlayVFX
        bForceVFXAtTargetLocation=false
        vFX=/Game/Effects/Fort_Effects/Gameplay/TimeTrial/NS_S29_MythicLabors_FootRace_Success
        effectLocationOffset=<0 0 0>
        effectScale=<1 1 1>
       }
       {
        _structType=/Script/FortniteGame.FortEventReaction_PlayAudio
        audio=/Game/Athena/Sounds/Quests/TimedTrial/Athena_Quest_TimeTrial_Success_Cue
        bForceSoundAtTargetLocation=false
        pitchMultiplier=1
        volumeMultiplier=1
        soundLocationOffset=<0 0 0>
       }
      ]
     }
     node {
      name=None
      _structType=/Script/EncountersRuntime.GrantReactionStateTreeTask
      bCanEditConsideredForCompletion=true
      bConsideredForCompletion=true
      bHasTaskCompletionDelegateDispatcher=false
      bTaskEnabled=true
      transitionHandlingPriority=Normal
      bindingsBatch {value=65535}
      executionRuntimeTemplateIndex {value=65535}
      instanceDataHandle {index=65535, source=None, stateHandle {index=65535}}
      instanceTemplateIndex {value=65535}
      outputBindingsBatch {value=65535}
     }
    }
   }
   {
    id=#39
    description=BP SDA STTask Complete State
    properties {
     iD=#40
     instanceObject=&self:StateTreeEditorData_0.StateTreeState_7.BP_SDA_STTask_CompleteState_C_0
     node {
      name=None
      _structType=&StateTreeBlueprintTaskWrapper
      taskClass=/StateDrivenActor/Tasks/BP_SDA_STTask_CompleteState.BP_SDA_STTask_CompleteState_C
     }
    }
   }
  ]
  transitions [{id=#41, trigger=ON_STATE_COMPLETED, state {id=#15, name=Root, linkType=SUCCEEDED}}]
 }
 Wait for player to enter QuestScenarioVolume {
  id=#16
  type=STATE
  tasks [{
    id=#42
    description="Scenario [Player Count Transition]"
    properties {
     iD=#43
     instance {
      _structType=/Script/EncountersRuntime.EncounterPlayerCountTransitionStateTreeTaskInstanceData
      bOnlyTriggerOnce=true
      bShouldTransition=true
      bTriggered=false
      comparisonOperator=GreaterThan
      userActor=null
      value=0
      customVerb {amount=1, customVerbTag {tagName=None}, subjects [], values []}
      onTriggered {iD=#26}
     }
     node {
      name=None
      _structType=/Script/EncountersRuntime.EncounterPlayerCountTransitionStateTreeTask
      bCanEditConsideredForCompletion=true
      bConsideredForCompletion=true
      bHasTaskCompletionDelegateDispatcher=false
      bTaskEnabled=true
      transitionHandlingPriority=Normal
      bindingsBatch {value=65535}
      executionRuntimeTemplateIndex {value=65535}
      instanceDataHandle {index=65535, source=None, stateHandle {index=65535}}
      instanceTemplateIndex {value=65535}
      outputBindingsBatch {value=65535}
      transitionTo {name=None, fallback=None, iD=#26, linkType=NextState, stateHandle {index=65535}}
     }
    }
   }]
  transitions []
 }
 Spire active invulnerable {
  id=#17
  type=STATE
  tasks [
   {
    id=#44
    description=Grant Reactions - Spire invulnerable
    properties {
     iD=#45
     instance {
      _structType=&GrantReactionStateTreeTaskInstanceData
      grantType=GrantToTargetActor
      targetActor=null
      userActor=null
      reactionEntries [{rowValue {
         bUseDataRegistryId=false
         dataRegistryId=""
         dataTableRow {
          dataTable=&/DT_QuestScenario_Spire_Minor_Master_Scenario_1_Reactions
          rowName=SpireActiveInvulnerable
         }
        }}]
      reactions []
     }
     node {
      name=Grant Reactions - Spire invulnerable
      _structType=&GrantReactionStateTreeTask
      bCanEditConsideredForCompletion=true
      bConsideredForCompletion=true
      bHasTaskCompletionDelegateDispatcher=false
      bTaskEnabled=true
      transitionHandlingPriority=Normal
      bindingsBatch {value=65535}
      executionRuntimeTemplateIndex {value=65535}
      instanceDataHandle {index=65535, source=None, stateHandle {index=65535}}
      instanceTemplateIndex {value=65535}
      outputBindingsBatch {value=65535}
     }
    }
   }
   {
    id=#46
    description=BP SDA STTask Complete State
    properties {
     iD=#47
     instanceObject=&self:StateTreeEditorData_0.StateTreeState_9.BP_SDA_STTask_CompleteState_C_0
     node {
      name=None
      _structType=&StateTreeBlueprintTaskWrapper
      taskClass=&BP_SDA_STTask_CompleteState_C
     }
    }
   }
  ]
  transitions [{id=#48, trigger=ON_STATE_COMPLETED, state {id=#15, name=Root, linkType=NEXT_STATE}}]
 }
 Set combatants hostile {
  id=#18
  type=STATE
  tasks [
   {
    id=#49
    description="LWM Encounter [Modify Actor Faction]"
    properties {
     iD=#50
     instance {
      _structType=/Script/EncountersRuntime.LWMEncounterModifyActorFactionStateTreeTaskInstanceData
      modificationAction=ExclusiveSet
      userActor=null
      encounterQuery {autoDescription="", tokenStreamVersion=0, userDescription="", queryTokenStream [], tagDictionary []}
      factionTag {tagName=Athena.Faction.CherrySmoke.Hostile}
     }
     node {
      name=None
      _structType=/Script/EncountersRuntime.LWMEncounterModifyActorFactionStateTreeTask
      bCanEditConsideredForCompletion=true
      bConsideredForCompletion=true
      bHasTaskCompletionDelegateDispatcher=false
      bTaskEnabled=true
      transitionHandlingPriority=Normal
      bindingsBatch {value=65535}
      executionRuntimeTemplateIndex {value=65535}
      instanceDataHandle {index=65535, source=None, stateHandle {index=65535}}
      instanceTemplateIndex {value=65535}
      outputBindingsBatch {value=65535}
     }
    }
   }
   {
    id=#51
    description=BP SDA STTask Complete State
    properties {
     iD=#52
     instanceObject=&self:StateTreeEditorData_0.StateTreeState_4.BP_SDA_STTask_CompleteState_C_0
     node {
      name=None
      _structType=&StateTreeBlueprintTaskWrapper
      taskClass=&BP_SDA_STTask_CompleteState_C
     }
    }
   }
  ]
  transitions [{id=#53, trigger=ON_STATE_COMPLETED, state {id=#15, name=Root, linkType=NEXT_STATE}}]
 }
 Wait for SubScenario completion {
  id=#19
  type=STATE
  tasks [{
    id=#54
    description=Scenario Verb Transition
    properties {
     iD=#55
     instance {
      _structType=&EncounterVerbTransitionStateTreeTaskInstanceData
      bIsTriggered=false
      bUseAdvancedObjectives=false
      objectiveValueThreshold=1
      userActor=null
      objectiveVerbs [{
        _structType=/Script/EncountersRuntime.FortVerbFilter_EncounterCompleted
        successState=Any
        author {
         bCanGainProgressForThisObjective=true
         tagQuery {autoDescription="", tokenStreamVersion=0, userDescription="", queryTokenStream [], tagDictionary []}
        }
        context {tagQuery {autoDescription="", tokenStreamVersion=0, userDescription="", queryTokenStream [], tagDictionary []}}
        encounterItemDef {tagQuery {autoDescription="", tokenStreamVersion=0, userDescription="", queryTokenStream [], tagDictionary []}}
        encounterVolumeTags {tagQuery {autoDescription="", tokenStreamVersion=0, userDescription="", queryTokenStream [], tagDictionary []}}
        faction {tagQuery {autoDescription="", tokenStreamVersion=0, userDescription="", queryTokenStream [], tagDictionary []}}
        participants {
         bCanGainProgressForThisObjective=true
         tagQuery {autoDescription="", tokenStreamVersion=0, userDescription="", queryTokenStream [], tagDictionary []}
        }
       }]
      objectives []
      onTriggered {iD=#26}
     }
     node {
      name=None
      _structType=&EncounterVerbTransitionStateTreeTask
      bCanEditConsideredForCompletion=true
      bConsideredForCompletion=true
      bHasTaskCompletionDelegateDispatcher=false
      bTaskEnabled=true
      transitionHandlingPriority=Normal
      bindingsBatch {value=65535}
      executionRuntimeTemplateIndex {value=65535}
      instanceDataHandle {index=65535, source=None, stateHandle {index=65535}}
      instanceTemplateIndex {value=65535}
      outputBindingsBatch {value=65535}
      transitionTo {name=None, fallback=None, iD=#26, linkType=NextState, stateHandle {index=65535}}
     }
    }
   }]
  transitions []
 }
 "Spire Vulnerable, Objective: Destroy Spire, wait X seconds before respawning reinforcements" {
  id=#20
  type=STATE
  tasks [
   {
    id=#56
    description=Grant Reactions - Spire vulnerable
    properties {
     iD=#57
     instance {
      _structType=&GrantReactionStateTreeTaskInstanceData
      grantType=GrantToTargetActor
      targetActor=null
      userActor=null
      reactionEntries [{rowValue {
         bUseDataRegistryId=false
         dataRegistryId=""
         dataTableRow {
          dataTable=&DT_QuestScenario_Spire_Minor_Master_Scenario_1_Reactions
          rowName=SpireVulnerable
         }
        }}]
      reactions []
     }
     node {
      name=Grant Reactions - Spire vulnerable
      _structType=&GrantReactionStateTreeTask
      bCanEditConsideredForCompletion=true
      bConsideredForCompletion=true
      bHasTaskCompletionDelegateDispatcher=false
      bTaskEnabled=true
      transitionHandlingPriority=Normal
      bindingsBatch {value=65535}
      executionRuntimeTemplateIndex {value=65535}
      instanceDataHandle {index=65535, source=None, stateHandle {index=65535}}
      instanceTemplateIndex {value=65535}
      outputBindingsBatch {value=65535}
     }
    }
   }
   {
    id=#58
    description=Grant Quest
    properties {
     iD=#59
     instance {
      _structType=/Script/EncountersRuntime.GrantQuestStateTreeTaskInstanceData
      grantType=GrantToPlayspace
      questToGrant=/SpiresGameplay/Scenarios/Spire_Minor/Master_Scenario/Spire_Minor_Master_Scenario_1/Quests/PlaySpace_EncounterQuest_Spire_Minor_Master_Scenario_1_03
      targetActor=null
      userActor=null
     }
     node {
      name=None
      _structType=/Script/EncountersRuntime.GrantQuestStateTreeTask
      bCanEditConsideredForCompletion=true
      bConsideredForCompletion=true
      bHasTaskCompletionDelegateDispatcher=false
      bTaskEnabled=false
      transitionHandlingPriority=Normal
      bindingsBatch {value=65535}
      executionRuntimeTemplateIndex {value=65535}
      instanceDataHandle {index=65535, source=None, stateHandle {index=65535}}
      instanceTemplateIndex {value=65535}
      outputBindingsBatch {value=65535}
     }
    }
   }
   {
    id=#60
    description=Delay 25.0
    properties {
     iD=#61
     instance {_structType=&StateTreeDelayTaskInstanceData, duration=25}
     node {name=None, _structType=&StateTreeDelayTask}
    }
   }
  ]
  transitions [{id=#62, trigger=ON_STATE_COMPLETED, state {id=#15, name=Root, linkType=NEXT_STATE}}]
 }
 1st loop wait for 0 combatants alive {
  id=#21
  type=STATE
  children [#63, #64, #65]
  tasks [{
    id=#66
    description="LWM Encounter [Get Remaining Count]"
    properties {
     iD=#67
     instance {
      _structType=/Script/EncountersRuntime.EncounterGetRemainingCountStateTreeTaskInstanceData
      userActor=null
      value=0
      encounterQuery {
       autoDescription=" ANY( QuestEncounter.Spire_Minor_Combat_1.AI.Goal.Assault.Wave.1 )"
       tokenStreamVersion=0
       userDescription=""
       queryTokenStream [0,1,1,1,0]
       tagDictionary [{tagName=QuestEncounter.Spire_Minor_Combat_1.AI.Goal.Assault.Wave.1}]
      }
     }
     node {
      name=None
      _structType=/Script/EncountersRuntime.EncounterGetRemainingCountStateTreeTask
      bCanEditConsideredForCompletion=true
      bConsideredForCompletion=true
      bHasTaskCompletionDelegateDispatcher=false
      bTaskEnabled=true
      transitionHandlingPriority=Normal
      bindingsBatch {value=65535}
      executionRuntimeTemplateIndex {value=65535}
      instanceDataHandle {index=65535, source=None, stateHandle {index=65535}}
      instanceTemplateIndex {value=65535}
      outputBindingsBatch {value=65535}
     }
    }
   }]
  transitions []
 }
 2nd loop wait for 0 combatants alive {
  id=#22
  type=STATE
  tasks [{
    id=#68
    description="LWM Encounter [Remaining Count Transition]"
    properties {
     iD=#69
     instance {
      _structType=/Script/EncountersRuntime.LWMEncounterRemainingCountTransitionStateTreeTaskInstanceData
      bOnlyTriggerOnce=true
      bShouldTransition=true
      bTriggered=false
      userActor=null
      value=0
      customVerb {amount=1, customVerbTag {tagName=None}, subjects [], values []}
      encounterQuery {
       autoDescription=" ANY( QuestEncounter.Spire_Minor_Combat_1.AI.Goal.Assault.Wave.1 )"
       tokenStreamVersion=0
       userDescription=""
       queryTokenStream [0,1,1,1,0]
       tagDictionary [{tagName=QuestEncounter.Spire_Minor_Combat_1.AI.Goal.Assault.Wave.1}]
      }
      onTriggered {iD=#26}
     }
     node {
      name=None
      _structType=/Script/EncountersRuntime.LWMEncounterRemainingCountTransitionStateTreeTask
      bCanEditConsideredForCompletion=true
      bConsideredForCompletion=true
      bHasTaskCompletionDelegateDispatcher=false
      bTaskEnabled=true
      transitionHandlingPriority=Normal
      bindingsBatch {value=65535}
      executionRuntimeTemplateIndex {value=65535}
      instanceDataHandle {index=65535, source=None, stateHandle {index=65535}}
      instanceTemplateIndex {value=65535}
      outputBindingsBatch {value=65535}
      transitionTo {name=None, fallback=None, iD=#26, linkType=NextState, stateHandle {index=65535}}
     }
    }
   }]
  transitions []
 }
 "Delay before respawn, transition to spawn reinforcements" {
  id=#23
  type=STATE
  tasks [{
    id=#70
    description=Delay 20.0
    properties {
     iD=#71
     instance {_structType=&StateTreeDelayTaskInstanceData, duration=20}
     node {name=None, _structType=&StateTreeDelayTask}
    }
   }]
  transitions [{id=#72, trigger=ON_STATE_COMPLETED, state {id=#63, name=HACK Delay}}]
 }
 HACK Delay {
  id=#63
  type=STATE
  tasks [{
    id=#73
    description=Delay 0.01
    properties {
     iD=#74
     instance {_structType=&StateTreeDelayTaskInstanceData, duration=0.01}
     node {name=None, _structType=&StateTreeDelayTask}
    }
   }]
  transitions [{id=#75, trigger=ON_STATE_COMPLETED, state {id=#15, name=Root, linkType=NEXT_STATE}}]
 }
 Spawn reinforcements {
  id=#64
  type=STATE
  tasks [
   {
    id=#76
    description="LWM Encounter [Dynamic Start]"
    properties {
     iD=#77
     instance {
      _structType=&LWMEncounterDynamicStartStateTreeTaskInstanceData
      bAttachVolumeToSpawn=false
      bOverrideAnchorPosition=false
      bOverrideDelayBetweenEvents=false
      bOverrideEnvQuery=false
      bOverrideStageActorMaxCount=false
      bOverrideStageActorMaxSpawnedCount=false
      bPauseOnExit=false
      bSpecifyFaction=true
      bUseVolumeAsAnchor=true
      bWaitForEncounterDoneSpawning=false
      delayBetweenEventsMax=0
      delayBetweenEventsMin=0
      dynamicEmptyEncounter=&Lager_Encounter_Dynamic_Empty
      faction=&Faction_Spires
      instigator=null
      owningActor=null
      spawnAroundEnvironmentQuery=null
      spawnAroundEnvironmentQueryRunMode=AllMatching
      spawnAroundNavigationSourceOverride=""
      stageActorMaxCount=1
      stageActorMaxSpawnedCount=-1
      userActor=null
      actorTags {gameplayTags []}
      anchorPointTagQuery {autoDescription="", tokenStreamVersion=0, userDescription="", queryTokenStream [], tagDictionary []}
      dynamicEncounterEntries [{
        actorCountOverrideMax=1
        actorCountOverrideMin=1
        bCreateRandomSingleActorEvents=false
        bOverrideActorCount=true
        bUseEncounterTagQuery=true
        bUseEventTemplate=true
        debugName=Test SubScenario Combatant
        maxConcurrentSpawns=1
        noActorLimitEventCount=1
        spawnCount=1
        actorDescriptions []
        encounterTagQuery {autoDescription="", tokenStreamVersion=0, userDescription="", queryTokenStream [], tagDictionary []}
        eventTemplate {
         activationGamePhaseStep=None
         activationMaxDelay=0
         activationMinDelay=0
         activationSafeZoneIndex=0
         bCanSpawnWithoutMatchingPlayerDistance=false
         bCheckForActorDensity=true
         bOnlyInNextSafeZone=false
         bOnlyInNextSafeZoneIfStormShrinking=false
         bOnlyInSafeZone=false
         bShouldProcessEventRequestImmediately=false
         deactivationMaxDelay=-1
         deactivationMinDelay=-1
         maxActorDensityOverride=-1
         maxDistanceToPlayer=-1
         minDistanceToPlayer=-1
         randomActivationChance=100
         weight=1
         calendarEventConditions {isActiveWithoutSeasonalManager=false, conditions []}
         canSpawnCondition {description=Empty, schemaClass=&FortAthenaLivingWorldConditionSchema, editableConditions [], sharedDefinition {}}
         despawnCondition {
          bCanBeDespawned=false
          bCanDespawnInTheStorm=false
          bCanDespawnWhenInCombatWithPlayer=false
          canStartDespawningPhaseStep=None
          canStartDespawningSafeZonePhase=-1
          maxStimulusAgeToBeConsideredInCombat=30
          minDistanceToPlayer=-1
          timeAsCandidateBeforeDespawn=10
         }
         despawnConditions []
         eventDatas []
         eventInstanceMaxCount [-1]
         eventInstanceMaxSpawnedCount [-1]
         instantDeathMaxRespawnTime []
         instantDeathMinRespawnTime []
         isEnabled {registryType=None, value=1, curve {curveTable=null, rowName=None}}
         maxCount [3]
         maxRespawnTime []
         maxSpawnedCount [3]
         minRespawnTime []
         playlistTagQuery {autoDescription="", tokenStreamVersion=0, userDescription="", queryTokenStream [], tagDictionary []}
         tagDensityConditions []
        }
        pointProviderTagQueryOverride {autoDescription="", tokenStreamVersion=0, userDescription="", queryTokenStream [], tagDictionary []}
       }]
      encounterIdentifierTags {gameplayTags [{tagName=QuestEncounter.Spire_Minor_Combat_1.AI.Goal.Assault.Wave.1}]}
      leashRadiusInner {registryType=None, value=1000, curve {curveTable=null, rowName=None}}
      leashRadiusOuter {registryType=None, value=3000, curve {curveTable=null, rowName=None}}
      pointProviderTagQuery {
       autoDescription=" ANY( QuestEncounter.Spire_Minor_Combat_1.Location.SpawnPoint.Wave )"
       tokenStreamVersion=0
       userDescription=Spawning will select a point for each combatant that matches this tag
       queryTokenStream [0,1,1,1,0]
       tagDictionary [{tagName=QuestEncounter.Spire_Minor_Combat_1.Location.SpawnPoint.Wave}]
      }
      spawnActions [{spawnActionClass=&BP_PlayVFX_SpawnFX_C, spawnActionTag {tagName=None}}]
      spawnTimeout {registryType=None, value=-1, curve {curveTable=null, rowName=None}}
     }
     node {
      name=None
      _structType=&LWMEncounterDynamicStartStateTreeTask
      bCanEditConsideredForCompletion=true
      bConsideredForCompletion=true
      bHasTaskCompletionDelegateDispatcher=false
      bTaskEnabled=true
      transitionHandlingPriority=Normal
      bindingsBatch {value=65535}
      executionRuntimeTemplateIndex {value=65535}
      instanceDataHandle {index=65535, source=None, stateHandle {index=65535}}
      instanceTemplateIndex {value=65535}
      outputBindingsBatch {value=65535}
     }
    }
   }
   {
    id=#78
    description=Delay 3.0
    properties {
     iD=#79
     instance {_structType=&StateTreeDelayTaskInstanceData, duration=3}
     node {name=None, _structType=&StateTreeDelayTask}
    }
   }
  ]
  transitions [{
    id=#80
    trigger=ON_STATE_COMPLETED
    state {id=#15, name=2nd loop wait for 0 combatants alive, linkType=NEXT_STATE}
   }]
 }
 Set AI Faction {
  id=#65
  type=STATE
  tasks [
   {
    id=#81
    description="LWM Encounter [Modify Actor Faction]"
    properties {
     iD=#82
     instance {
      _structType=&LWMEncounterModifyActorFactionStateTreeTaskInstanceData
      modificationAction=ExclusiveSet
      userActor=null
      encounterQuery {autoDescription="", tokenStreamVersion=0, userDescription="", queryTokenStream [], tagDictionary []}
      factionTag {tagName=Athena.Faction.CherrySmoke.Hostile}
     }
     node {
      name=None
      _structType=&LWMEncounterModifyActorFactionStateTreeTask
      bCanEditConsideredForCompletion=true
      bConsideredForCompletion=true
      bHasTaskCompletionDelegateDispatcher=false
      bTaskEnabled=true
      transitionHandlingPriority=Normal
      bindingsBatch {value=65535}
      executionRuntimeTemplateIndex {value=65535}
      instanceDataHandle {index=65535, source=None, stateHandle {index=65535}}
      instanceTemplateIndex {value=65535}
      outputBindingsBatch {value=65535}
     }
    }
   }
   {
    id=#83
    description=BP SDA STTask Complete State
    properties {
     iD=#84
     instanceObject=&self:StateTreeEditorData_0.StateTreeState_15.BP_SDA_STTask_CompleteState_C_0
     node {
      name=None
      _structType=&StateTreeBlueprintTaskWrapper
      taskClass=&BP_SDA_STTask_CompleteState_C
     }
    }
   }
  ]
  transitions [{
    id=#85
    trigger=ON_STATE_COMPLETED
    state {id=#22, name=2nd loop wait for 0 combatants alive}
   }]
 }
]
